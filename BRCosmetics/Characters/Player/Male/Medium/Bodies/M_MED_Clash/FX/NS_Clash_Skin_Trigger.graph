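# Niagara system for the Clash character skin that fires a burst of small mesh tendrils on the body, skipped while in a vehicle.
# Particles split into spawn groups that index bones or sockets on a user-supplied skeletal mesh, following their position and orientation each update.
# Age-based curves grow and shrink the tendrils and animate dynamic material parameters.
# A PartHiddenBoneCheck script sets each particle's visibility tag from hand- and foot-hidden flags.
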NiagaraSystem /BRCosmetics/Characters/Player/Male/Medium/Bodies/M_MED_Clash/FX/NS_Clash_Skin_Trigger
dependencies {
 usedDataInterfaces [&NiagaraDataInterfaceVectorCurve, &NiagaraDataInterfaceCurve]
 usedDynamicInputs [
  &NormalizeFloat
  &VectorLength
  &SetBoolByFloatComparison
 ]
 usedModules [
  &self:SystemSpawnScript.SystemScriptSource.SystemScriptGraph.NiagaraNodeAssignment_0.SetVariables_#1
  &FN_Integrator
  &SystemState
 ]
}
emitters {Small_MeshTendrils_burst {
  emitterUpdateScript [
   &EmitterState(Life Cycle Mode=::NewEnumerator1, Loop Behavior=::NewEnumerator1, Loop Duration=1, UseLoopDelay=false)
   &SpawnBurst_Instantaneous(
    Spawn Count=&MakeCustomIntFromBool(Bool=$User.bisInVehicle, True INT=0, False INT=60)
    Use Spawn Probability=false
    UseLoopCountLimit=false
   )
  ]
  particleSpawnScript [
   &InitializeParticle(
    Color=<rgba 0.276 0.276 0.276 1>
    Mesh Scale=&Multiply_VectorByFloat(
     Vector=&RandomRangeVector(Minimum=<0.7 0.7 1.1>, Maximum=<1 0.8 1.1>)
     Float=&RandomRangeFloat(Minimum=0.9, Maximum=1.2)
    )
    Mesh Scale Mode=::NewEnumerator4
    UsePositionOffset=false
    Write Material Random=true
   ) {module=InitializeParticle001}
   &FN_SetSpawnGroup(Number of Spawn Groups=14)
   &InitialMeshOrientation(
    Rotation=&RandomRangeVector(Minimum=<-0.14 -0.07 0>, Maximum=<0.14 0.07 0>)
    Use Orientation Vector=true
    Use Rotation Vector=true
   ) {module=InitialMeshOrientation001}
   /SetVariables(Particles.VisibilityTag=&RandomRangeInt(Minimum=0, Maximum=1)) {isSetParameters=true}
   &CameraOffset(Camera Offset Amount=&RandomRangeFloat(Minimum=-10, Maximum=-5))
  ]
  particleUpdateScript [
   /SetVariables(Particles.Mesh_Scale02:Vector3f=&Multiply_Vector(
     A=&Multiply_Vector(
      A=&Multiply_VectorByFloat(
       Vector=&RandomRangeVector(Minimum=<0.6 0.6 0.8>, Maximum=<0.8 0.8 1>)
       Float=&RandomRangeFloat(Minimum=0.9, Maximum=1.4)
      )
      B=&VectorFromCurve(
       VectorCurve=DataInterface() {propertyValues {
         XCurve {keys [<at 0 0 RCIM_Cubic>,<at 0.5 1 RCIM_Cubic>,<at 1 0 RCIM_Cubic>]}
         YCurve {keys [<at 0 0 RCIM_Cubic>,<at 0.5 1 RCIM_Cubic>,<at 1 0 RCIM_Cubic>]}
         ZCurve {keys [
           {arriveTangent=2.4741, interpMode=RCIM_Cubic, leaveTangent=2.4741, tangentMode=RCTM_User, time=0, value=0}
           <at 0.5 1 RCIM_Cubic>
           {arriveTangent=-2.7109, interpMode=RCIM_Cubic, leaveTangent=-2.7109, tangentMode=RCTM_User, time=1, value=0}
          ]}
        }}
       Scale Curve=&VectorFromFloat(Value=0.85)
      )
     )
     B=&VectorFromFloat(Value=&MakeCustomFloatFromBool(Bool=false, True Float=1, False Float=0))
    )) {isSetParameters=true}
   &ParticleState()
   &DynamicMaterialParameters(
    Index 0 Param 1=&FloatFromCurve(FloatCurve=DataInterface() {propertyValues {Curve {keys [<at 0 0 RCIM_Cubic>,<at 0.5 1 RCIM_Cubic>,<at 1 0 RCIM_Cubic>]}}}, Scale Curve=1.25)
    Index 0 Param 2=&FloatFromCurve(
     FloatCurve=DataInterface() {propertyValues {Curve {keys [<at 0 1 RCIM_Cubic>,<at 0.5 0.1 RCIM_Cubic>,<at 1 1 RCIM_Cubic>]}}}
     Scale Curve=3.5
    )
    Index 0 Param 4=1
    Module.Param0WriteEnabled=true
    Module.Param1WriteEnabled=true
    Module.Param2WriteEnabled=true
    Module.Param3WriteEnabled=true
   )
   &SkeletalMeshLocation(
    Apply Sampled Orientation to Mesh Particles=true
    Bone / Socket Index=$Particles.SpawnGroup
    Inherit Velocity=false
    MaskSpawnGroup=false
    Mesh Sampling Type=::NewEnumerator5
    Sampled Position Offset=&MakeVector(
     X=&RandomRangeFloat(Minimum=0, Maximum=10)
     Y=0
     Z=&RandomRangeFloat(Minimum=-2, Maximum=-2)
    )
    Sampled Position Offset Coordinate Space:ENiagaraMeshTransforms=::NewEnumerator3
    Skeletal Mesh=$User.SK_Mesh_Tendrils
    Socket Sampling Mode=::NewEnumerator5
    Use Sampled Mesh Attribute as Ribbon ID=false
   )
   &ScaleMeshSize(Scale Factor=&Multiply_Vector(
     A=<1 1 1>
     B=&Multiply_VectorByFloat(
      Vector=<1 1 1>
      Float=&FloatFromCurve(
       FloatCurve=DataInterface() {propertyValues {Curve {keys [
           {tangentMode=RCTM_User, time=0, value=0}
           <at 0.2 1>
           {arriveTangent=-0.0405, interpMode=RCIM_Cubic, leaveTangent=-0.0405, tangentMode=RCTM_Break, time=0.4, value=1}
           {arriveTangent=-2.6037, interpMode=RCIM_Cubic, leaveTangent=-8.0089, tangentMode=RCTM_Break, time=1, value=0}
          ]}}}
       CurveIndex=$Particles.Age
      )
     )
    )) {module=ScaleMeshSize001}
   /SetVariables(Particles.VisibilityTag=&self:PartHiddenBoneCheck(isHandHidden:NiagaraBool=$User.bisHandHidden, isFootHidden:NiagaraBool=$User.bisFootHidden, Selection:NiagaraInt32=$Particles.SpawnGroup)) {isSetParameters=true}
  ]
  rendererClasses [&NiagaraMeshRendererProperties]
  renderers [
   {rendererClass=&NiagaraMeshRendererProperties, rendererIndex=0}
   {rendererClass=&NiagaraMeshRendererProperties, rendererIndex=1}
  ]
 }}
summary {
 systemName=NS_Clash_Skin_Trigger
 emitters [{emitterName=Small_MeshTendrils_burst, rendererClasses [&NiagaraMeshRendererProperties]}]
 userVariables [
  User.bisFootHidden {type=NiagaraBool}
  User.bisHandHidden {type=NiagaraBool}
  User.bisInVehicle {type=NiagaraBool}
  User.SK_Mesh_Tendrils {
   type=NiagaraDataInterfaceSkeletalMesh
   defaultValue {
    struct=NiagaraExt_VariableValue_DataInterface
    value {dataInterface=&self:NiagaraDataInterfaceSkeletalMesh_2}
   }
  }
 ]
}
