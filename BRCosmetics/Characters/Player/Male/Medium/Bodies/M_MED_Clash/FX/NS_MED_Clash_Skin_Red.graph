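# Niagara skin effect for the red M_MED_Clash character body, adding flowing ribbons, socket trails and tentacle meshes.
# Bone-flow ribbons spiral along four bone selections; socket trails react to owner speed, and both hide with hidden hands or feet.
# Tentacles follow spine_04: kill-trigger versions animate through User parameters, while front-end duplicates loop on baked curves alongside socket-attached tendrils.
# User flags for preview, skydiving and vehicle state gate spawning.
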NiagaraSystem /BRCosmetics/Characters/Player/Male/Medium/Bodies/M_MED_Clash/FX/NS_MED_Clash_Skin_Red
dependencies {
 usedDataInterfaces [&NiagaraDataInterfaceCurve, &NiagaraDataInterfaceArrayFloat, &NiagaraDataInterfaceVectorCurve]
 usedDynamicInputs []
 usedModules [
  &self:SystemSpawnScript.SystemScriptSource.SystemScriptGraph.NiagaraNodeAssignment_0.SetVariables_#1
  &FN_Integrator
  &SystemState
 ]
}
emitters {
 Bone_Flow_Ribbon01 {
  emitterUpdateScript [
   &EmitterState(
    Life Cycle Mode=::NewEnumerator1
    Loop Delay=&RandomRangeFloat(Minimum=0, Maximum=1)
    Loop Duration=&RandomRangeFloat(Minimum=1, Maximum=2)
    Recalculate Duration Each Loop=true
    UseLoopDelay=true
   )
   &SpawnBurst_Instantaneous(Spawn Count=15, Use Spawn Probability=false, UseLoopCountLimit=false)
  ]
  particleSpawnScript [
   &FN_SetSpawnGroup(Inclusivity=true, Normalize Spawn Group=true, Particles Per Group=15, Pattern=::NewEnumerator0, Set Ribbon ID and Ribbon Link Order=true)
   &InitializeRibbon(
    Lifetime Max=1.5
    Lifetime Min=1.5
    Lifetime Mode:ENiagara_LifetimeMode=::NewEnumerator1
    UsePositionOffset=false
    Write Material Random=true
   )
   &CameraOffset(Camera Offset Amount=10)
   &DynamicMaterialParameters(
    Index 0 Param 1=0.5
    Index 0 Param 2=1
    Index 0 Param 3=1
    Index 0 Param 4=1
    Module.Param0WriteEnabled=true
    Module.Param1WriteEnabled=true
    Module.Param2WriteEnabled=true
    Module.Param3WriteEnabled=true
   )
  ]
  particleUpdateScript [
   &ParticleState()
   &FN_BoneFlow(
    Angle:NiagaraFloat=&Multiply_Float(
     A=$Particles.Normalized Spawn Group
     B=&Add_Float(
      A=&Multiply_Float(A=$Particles.Age, B=0.5)
      B=&Lerp_Float(A=1.5, B=2, Alpha=$Particles.Normalized Spawn Group)
     )
    )
    BoneID:NiagaraInt32=&ModuloInt(A=$Particles.SpawnGroup, B=4)
    Height:NiagaraFloat=&Multiply_Float(A=$Particles.Normalized Spawn Group, B=1.1)
    Radius:NiagaraFloat=&Multiply_Float(
     A=&FloatFromCurve(
      FloatCurve=DataInterface() {propertyValues {Curve {keys [<at 0 3 RCIM_Cubic>,<at 0.5 12.5 RCIM_Cubic>,<at 1 3>]}}}
      CurveIndex=$Particles.Normalized Spawn Group
      Scale Curve=&RandomRangeFloat(Minimum=0.7, Maximum=0.7)
     )
     B=&FloatFromCurve(
      FloatCurve=DataInterface() {propertyValues {Curve {keys [
          {interpMode=RCIM_Cubic, tangentMode=RCTM_User, time=0, value=0}
          {interpMode=RCIM_Cubic, tangentMode=RCTM_User, time=0.5, value=1}
          {tangentMode=RCTM_User, time=1, value=0.3}
         ]}}}
      Scale Curve=&RandomRangeFloat(Minimum=1, Maximum=1)
     )
    )
    SkeletalMesh:NiagaraDataInterfaceSkeletalMesh=$User.SK_Mesh
   )
   /SetVariables(Particles.RibbonFacing=$Output.FN_BoneFlow.SplineNormal) {isSetParameters=true, module=SetVariables_#2}
   &ScaleRibbonWidth(Ribbon Width Scale=&Multiply_Float(
     A=&Multiply_Float(
      A=&FloatFromCurve(
       FloatCurve=DataInterface() {propertyValues {Curve {keys [
           {interpMode=RCIM_Cubic, tangentMode=RCTM_User, time=0, value=0}
           <at 0.1 0>
           {interpMode=RCIM_Cubic, tangentMode=RCTM_Break, time=0.2, value=1}
           {arriveTangent=-1.5728, interpMode=RCIM_Cubic, leaveTangent=-1.5728, tangentMode=RCTM_User, time=1, value=0}
          ]}}}
       Scale Curve=&RandomRangeFloat(Minimum=0.7, Maximum=1)
      )
      B=1
     )
     B=&FloatFromCurve(
      FloatCurve=DataInterface() {propertyValues {Curve {keys [
          {interpMode=RCIM_Cubic, tangentMode=RCTM_User, time=0, value=0}
          {interpMode=RCIM_Cubic, tangentMode=RCTM_User, time=0.3, value=1}
          {arriveTangent=-1.923, interpMode=RCIM_Cubic, leaveTangent=-1.923, tangentMode=RCTM_User, time=1, value=0}
         ]}}}
      CurveIndex=$Particles.Normalized Spawn Group
     )
    ))
   &ScaleColor(
    Scale Alpha=&FloatFromCurve(FloatCurve=DataInterface() {propertyValues {Curve {keys [
         {interpMode=RCIM_Cubic, tangentMode=RCTM_User, time=0, value=0}
         <at 0.2 1>
         {interpMode=RCIM_Cubic, tangentMode=RCTM_User, time=0.8, value=1}
         {tangentMode=RCTM_User, time=1, value=0}
        ]}}})
    Scale RGB=&Multiply_VectorByFloat(
     Vector=<1 1 1>
     Float=&MakeCustomFloatFromBool(Bool=$User.bisFrontEndPreview, True Float=0.8, False Float=1.1)
    )
    ScaleA=true
    ScaleRGB=true
   )
   &SolveForcesAndVelocity(Clamp Velocity=false, Limit Acceleration=false)
   &ScaleRibbonWidth(Ribbon Width Scale=&SelectFloatFromArray(
     Direct Array Index=&self:PartHiddenPartitionCheck(
      is Hand Hidden:NiagaraBool=$User.bisHandHidden
      is Foot Hidden:NiagaraBool=$User.bisFootHidden
      Selection:NiagaraInt32=&ModuloInt(A=$Particles.SpawnGroup, B=4)
     )
     Float Selection Array=DataInterface() {propertyValues {FloatData [1,0]}}
    )) {module=ScaleRibbonWidth001}
  ]
  rendererClasses [&NiagaraRibbonRendererProperties]
  renderers [{rendererClass=&NiagaraRibbonRendererProperties}]
 }
 Bone_Flow_Ribbon02 {
  emitterUpdateScript [
   &EmitterState(
    Life Cycle Mode=::NewEnumerator1
    Loop Delay=&RandomRangeFloat(Minimum=0.5, Maximum=2)
    Loop Duration=&RandomRangeFloat(Minimum=1, Maximum=2)
    Recalculate Duration Each Loop=true
    UseLoopDelay=true
   )
   &SpawnBurst_Instantaneous(Spawn Count=15, Use Spawn Probability=false, UseLoopCountLimit=false)
  ]
  particleSpawnScript [
   &FN_SetSpawnGroup(Inclusivity=true, Normalize Spawn Group=true, Particles Per Group=15, Pattern=::NewEnumerator0, Set Ribbon ID and Ribbon Link Order=true)
   &InitializeRibbon(Lifetime Min=2, Lifetime Mode:ENiagara_LifetimeMode=::NewEnumerator1, UsePositionOffset=false, Write Material Random=true)
   &CameraOffset(Camera Offset Amount=10)
   &DynamicMaterialParameters(
    Index 0 Param 1=0.5
    Index 0 Param 2=1
    Index 0 Param 3=1
    Index 0 Param 4=1
    Module.Param0WriteEnabled=true
    Module.Param1WriteEnabled=true
    Module.Param2WriteEnabled=true
    Module.Param3WriteEnabled=true
   )
  ]
  particleUpdateScript [
   &ParticleState()
   &FN_BoneFlow(
    Angle:NiagaraFloat=&Multiply_Float(
     A=$Particles.Normalized Spawn Group
     B=&Add_Float(
      A=&Multiply_Float(A=$Particles.Age, B=0.5)
      B=&Lerp_Float(A=0.5, B=1, Alpha=$Particles.Normalized Spawn Group)
     )
    )
    BoneID:NiagaraInt32=&ModuloInt(A=$Particles.SpawnGroup, B=4)
    Height:NiagaraFloat=&Multiply_Float(A=$Particles.Normalized Spawn Group, B=0.8)
    Radius:NiagaraFloat=&Multiply_Float(
     A=&FloatFromCurve(
      FloatCurve=DataInterface() {propertyValues {Curve {keys [<at 0 3 RCIM_Cubic>,<at 0.5 13 RCIM_Cubic>,<at 1 3>]}}}
      CurveIndex=$Particles.Normalized Spawn Group
      Scale Curve=&RandomRangeFloat(Minimum=0.7, Maximum=0.7)
     )
     B=&FloatFromCurve(
      FloatCurve=DataInterface() {propertyValues {Curve {keys [
          {interpMode=RCIM_Cubic, tangentMode=RCTM_User, time=0, value=0}
          {interpMode=RCIM_Cubic, tangentMode=RCTM_User, time=0.5, value=1}
          {tangentMode=RCTM_User, time=1, value=0.3}
         ]}}}
      Scale Curve=&RandomRangeFloat(Minimum=1, Maximum=1.1)
     )
    )
    SkeletalMesh:NiagaraDataInterfaceSkeletalMesh=$User.SK_Mesh
   )
   /SetVariables(Particles.RibbonFacing=$Output.FN_BoneFlow.SplineNormal) {isSetParameters=true, module=SetVariables_#2}
   &ScaleRibbonWidth(Ribbon Width Scale=&Multiply_Float(
     A=&Multiply_Float(
      A=&FloatFromCurve(
       FloatCurve=DataInterface() {propertyValues {Curve {keys [
           {interpMode=RCIM_Cubic, tangentMode=RCTM_User, time=0, value=0}
           <at 0.1 0>
           {interpMode=RCIM_Cubic, tangentMode=RCTM_Break, time=0.2, value=1}
           {arriveTangent=-1.5728, interpMode=RCIM_Cubic, leaveTangent=-1.5728, tangentMode=RCTM_User, time=1, value=0}
          ]}}}
       Scale Curve=&RandomRangeFloat(Minimum=0.7, Maximum=1)
      )
      B=1
     )
     B=&FloatFromCurve(
      FloatCurve=DataInterface() {propertyValues {Curve {keys [
          {interpMode=RCIM_Cubic, tangentMode=RCTM_User, time=0, value=0}
          {interpMode=RCIM_Cubic, tangentMode=RCTM_User, time=0.3, value=1}
          {arriveTangent=-1.923, interpMode=RCIM_Cubic, leaveTangent=-1.923, tangentMode=RCTM_User, time=1, value=0}
         ]}}}
      CurveIndex=$Particles.Normalized Spawn Group
     )
    ))
   &ScaleColor(
    Scale Alpha=&FloatFromCurve(FloatCurve=DataInterface() {propertyValues {Curve {keys [
         {interpMode=RCIM_Cubic, tangentMode=RCTM_User, time=0, value=0}
         <at 0.2 1>
         {interpMode=RCIM_Cubic, tangentMode=RCTM_User, time=0.8, value=1}
         {tangentMode=RCTM_User, time=1, value=0}
        ]}}})
    Scale RGB=&Multiply_VectorByFloat(
     Vector=<1 1 1>
     Float=&MakeCustomFloatFromBool(Bool=$User.bisFrontEndPreview, True Float=0.8, False Float=1.1)
    )
    ScaleA=true
    ScaleRGB=true
   )
   &SolveForcesAndVelocity(Clamp Velocity=false, Limit Acceleration=false)
   &ScaleRibbonWidth(Ribbon Width Scale=&SelectFloatFromArray(
     Direct Array Index=&self:PartHiddenPartitionCheck(
      is Hand Hidden:NiagaraBool=$User.bisHandHidden
      is Foot Hidden:NiagaraBool=$User.bisFootHidden
      Selection:NiagaraInt32=&ModuloInt(A=$Particles.SpawnGroup, B=4)
     )
     Float Selection Array=DataInterface() {propertyValues {FloatData [1,0]}}
    )) {module=ScaleRibbonWidth001}
  ]
  rendererClasses [&NiagaraRibbonRendererProperties]
  renderers [{rendererClass=&NiagaraRibbonRendererProperties}]
 }
 FN_AttachedRibbon_v3003 {
  emitterUpdateScript [
   &EmitterState(
    Enable Distance Culling=true
    Min Distance Response=::SleepAndLetParticlesFinish
    MinDistance=180
    Module.bUseMaxDistance=false
    Module.bUseMinDistance=true
    Scalability Mode=::NewEnumerator1
   )
   /Game/Effects/Niagara/Modules/Ribbons/FN_SpawnRibbon_v3(
    Number of Ribbons:NiagaraInt32=&Multiply_Int(
     A=8
     B=&MakeCustomIntFromBool(
      Bool=$User.IsSkydiving
      True INT=0
      False INT=&MakeCustomIntFromBool(Bool=$User.bisInVehicle, True INT=0, False INT=1)
     )
    )
    Skeleton Mesh:NiagaraDataInterfaceSkeletalMesh=$User.SkeletalMesh_Trail_Scoket
    SpawnRate:NiagaraFloat=&Multiply_Float(
     A=20
     B=&MakeCustomFloatFromBool(
      Bool=&SetBoolByFloatComparison(A=&VectorLength(Vector=$Engine.Owner.Velocity), B=1200)
      True Float=0
      False Float=1
     )
    )
   )
  ]
  particleSpawnScript [
   /Game/Effects/Niagara/Modules/Ribbons/FN_InitializeRibbon_v4(
    Enable Skeletal Position=true
    Lifetime:NiagaraFloat=&Lerp_Float(
     A=0.8
     B=0.01
     Alpha=&Divide_Float(A=&VectorLength(Vector=$Engine.Owner.Velocity), B=2000)
    )
    Ribbon Width=10
    SkeletalMesh:NiagaraDataInterfaceSkeletalMesh=$User.SkeletalMesh_Trail_Scoket
   )
   &AddVelocityFromPoint(Module.Use Falloff Distance=false, Origin Offset=<0 0 100>, Velocity Strength=40)
  ]
  particleUpdateScript [
   &ParticleState()
   &InheritSourceMovement(
    Applied Position Falloff Distance=100
    Applied Position Scale=<0.75 0.75 0.75>
    ApplyForce=false
    ApplyPosition=true
    ApplyPositionFalloff=false
   )
   /SetVariables(Particles.DistanceFromBody:NiagaraFloat=&NormalizeFloat(
     Value=&VectorLength(Vector=&Subtract_Vector(A=Linked() {linkedVariable {name=Particles.Position, type=NiagaraPosition}}, B=$Particles.FN_UpdateRibbon_v4.SampledPosition))
     Min=0
     Max=50
    )) {isSetParameters=true}
   /Niagara/Modules/Update/Forces/CurlNoiseForce(
    Noise Strength=&FloatFromCurve(
     FloatCurve=DataInterface() {propertyValues {Curve {keys [<at 0.1 25>,<at 1 2000>]}}}
     CurveIndex=&Divide_Float(A=&VectorLength(Vector=$Engine.Owner.Velocity), B=900)
    )
    Pan Noise=true
    Pan Noise Field=<0.05 0.05 0.05>
    Random Seed:NiagaraInt32=&RandomRangeInt(Minimum=0, Maximum=1)
   )
   &Drag(Drag=0.9, Use Linear Drag=true, Use Rotational Drag=false)
   &SolveForcesAndVelocity(Clamp Velocity=false, Limit Acceleration=false)
   /Game/Effects/Niagara/Modules/Ribbons/FN_UpdateRibbon_v4(Enable Skeletal Position=true, SkeletalMesh:NiagaraDataInterfaceSkeletalMesh=$User.SkeletalMesh_Trail_Scoket)
   &ScaleRibbonWidth(Ribbon Width Scale=&Multiply_Float(
     A=&Multiply_Float(
      A=&Multiply_Float(
       A=&Multiply_Float(
        A=&MakeCustomFloatFromBool(
         Bool=&SetBoolByFloatComparison(A=&VectorLength(Vector=$Engine.Owner.Velocity), B=1500)
         True Float=0
         False Float=1
        )
        B=&FloatFromCurve(
         FloatCurve=DataInterface() {propertyValues {Curve {keys [{arriveTangent=19.0097, interpMode=RCIM_Cubic, leaveTangent=-1.3891, tangentMode=RCTM_User, time=0, value=1}, <at 1 0 RCIM_Cubic>]}}}
         Scale Curve=0.85
        )
       )
       B=&FloatFromCurve(
        FloatCurve=DataInterface() {propertyValues {Curve {keys [{arriveTangent=-7.8567, time=0, value=0}, <at 0.4 1>]}}}
        CurveIndex=$System.FN_Integrator.Value
       )
      )
      B=&FloatFromCurve(
       FloatCurve=DataInterface() {propertyValues {Curve {keys [
           {arriveTangent=-0.3908, interpMode=RCIM_Cubic, time=0, value=1}
           {arriveTangent=-1, interpMode=RCIM_Cubic, leaveTangent=-1, time=0.7899, value=0.7951}
           {interpMode=RCIM_Cubic, leaveTangent=-2.5446, time=1, value=0}
          ]}}}
       CurveIndex=$Particles.DistanceFromBody
      )
     )
     B=&SelectFloatFromArray(
      Direct Array Index=&self:PartHiddenTrailSocket(isHandHidden:NiagaraBool=$User.bisHandHidden, isFootHidden:NiagaraBool=$User.bisFootHidden, Selection:NiagaraInt32=$Particles.SpawnGroup)
      Float Selection Array=DataInterface() {propertyValues {FloatData [1,0]}}
     )
    ))
   &CameraOffset(Camera Offset Amount=15)
   &DynamicMaterialParameters(
    Index 0 Param 1=&FloatFromCurve(FloatCurve=DataInterface() {propertyValues {Curve {keys [<at 0 1>,<at 1 0>]}}})
    Index 0 Param 2=$Particles.MaterialRandom
    Index 0 Param 3=&Lerp_Float(A=1, B=2, Alpha=$Particles.PlayerSpeed)
    Index 0 Param 4=1
    Module.Param0WriteEnabled=true
    Module.Param1WriteEnabled=true
    Module.Param2WriteEnabled=false
    Module.Param3WriteEnabled=true
   )
   &ScaleRibbonWidth(Ribbon Width Scale=&SelectFloatFromArray(
     Direct Array Index=&self:PartHiddenTrailSocket(isHandHidden:NiagaraBool=$User.bisHandHidden, isFootHidden:NiagaraBool=$User.bisFootHidden, Selection:NiagaraInt32=$Particles.SpawnGroup)
     Float Selection Array=DataInterface() {propertyValues {FloatData [0,1]}}
    )) {enabled=false, module=ScaleRibbonWidth001}
  ]
  rendererClasses [&NiagaraRibbonRendererProperties]
  renderers [{rendererClass=&NiagaraRibbonRendererProperties}]
 }
 Frontend_Kill_Trigger_Big_Tentacles001 {
  emitterUpdateScript [
   &EmitterState(Inactive Response=::NewEnumerator1, Life Cycle Mode=::NewEnumerator1, Loop Delay=1.5, Loop Duration=3, UseLoopDelay=true)
   &SpawnBurst_Instantaneous(
    Spawn Count=&MakeCustomIntFromBool(
     Bool=$User.bisFrontEndPreview
     True INT=&MakeCustomIntFromBool(Bool=$User.IsSkydiving, True INT=0, False INT=1)
     False INT=0
    )
    Use Spawn Probability=false
    UseLoopCountLimit=false
   )
  ]
  particleSpawnScript [
   &InitializeParticle(Color Mode=::NewEnumerator0, Lifetime=2.5, Mesh Scale Mode=::NewEnumerator4, UsePositionOffset=false, Write Material Random=true)
   &InitialMeshOrientation(Mesh Orientation Mode=::NewEnumerator2, Rotation=<0 0 -0.25>, Use Rotation Vector=true)
   &InitialMeshOrientation(Mesh Orientation Mode=::NewEnumerator2, Rotation=<0 0.2 0>, Use Rotation Vector=true) {module=InitialMeshOrientation001}
   &CameraOffset()
  ]
  particleUpdateScript [
   &ParticleState()
   &SkeletalMeshLocation(
    Bone Sampling Mode=::NewEnumerator4
    MaskSpawnGroup=false
    Sampled Position Offset=&MakeVector(
     X=0
     Y=&Lerp_Float(
      A=0
      B=-20
      Alpha=&FloatFromCurve(FloatCurve=DataInterface() {propertyValues {Curve {keys [
           {interpMode=RCIM_Cubic, tangentMode=RCTM_User, time=0, value=0}
           <at 0.15 1>
           {interpMode=RCIM_Cubic, tangentMode=RCTM_User, time=0.8, value=1}
           {tangentMode=RCTM_User, time=1, value=0}
          ]}}})
     )
     Z=0
    )
    Skeletal Mesh=DataInterface() {propertyValues {
      DefaultMesh=null
      ExcludeBoneName=root
      OptionalComponentClass=null
      PreviewMesh=/BRCosmetics/Characters/Player/Male/Medium/Bodies/M_MED_Clash/Meshes/M_MED_Clash
      SkinningMode=SkinOnTheFly
      SoftSourceActor=null
      SourceMode=Default
      UvSetIndex=0
      WholeMeshLOD=-1
      bExcludeBone=false
      bReadDeformedGeometry=true
      bRequireCurrentFrameData=true
      ComponentTags []
      FilteredBones [spine_04]
      FilteredSockets []
      MeshUserParameter {parameter {name=None, typeDefHandle {registeredTypeIndex=1}, varData []}}
      SamplingRegions []
     }}
   )
   &DynamicMaterialParameters(
    Index 0 Param 1=0.5
    Index 0 Param 2=1.5
    Index 0 Param 3=1.5
    Index 0 Param 4=&FloatFromCurve(FloatCurve=DataInterface() {propertyValues {Curve {keys [
         {interpMode=RCIM_Cubic, tangentMode=RCTM_User, time=0, value=0}
         <at 0.1 0>
         {interpMode=RCIM_Cubic, tangentMode=RCTM_User, time=0.75, value=0}
         {tangentMode=RCTM_User, time=1, value=1}
        ]}}})
    Index 1 Param 3=1
    Index 1 Param 4=0.5
    Module.Param0WriteEnabled=true
    Module.Param0WriteEnabled1=true
    Module.Param1WriteEnabled=true
    Module.Param1WriteEnabled1=true
    Module.Param2WriteEnabled=true
    Module.Param2WriteEnabled1=true
    Module.Param3WriteEnabled=true
    Module.Param3WriteEnabled1=true
    Write Parameter Index 1=true
   )
   &ScaleMeshSize(Scale Factor=&Multiply_VectorByFloat(
     Vector=&Multiply_VectorByFloat(Vector=<1 1 1>, Float=1)
     Float=&FloatFromCurve(FloatCurve=DataInterface() {propertyValues {Curve {keys [
          {interpMode=RCIM_Cubic, tangentMode=RCTM_User, time=0, value=0}
          <at 0.15 1>
          {interpMode=RCIM_Cubic, tangentMode=RCTM_User, time=0.8, value=1}
          {tangentMode=RCTM_User, time=1, value=0}
         ]}}})
    )) {module=ScaleMeshSize001}
   /SetVariables(Particles.Mesh02_Size:Vector3f=&Multiply_Vector(
     A=<1 1 1>
     B=&VectorFromFloat(Value=&FloatFromCurve(FloatCurve=DataInterface() {propertyValues {Curve {keys [
           {interpMode=RCIM_Cubic, tangentMode=RCTM_User, time=0.05, value=0}
           <at 0.2 1>
           {interpMode=RCIM_Cubic, tangentMode=RCTM_User, time=0.8, value=1}
           {tangentMode=RCTM_User, time=1, value=0}
          ]}}}))
    )) {isSetParameters=true, module=SetVariables_#3}
  ]
  rendererClasses [&NiagaraMeshRendererProperties]
  renderers [
   {rendererClass=&NiagaraMeshRendererProperties, rendererIndex=0}
   {rendererClass=&NiagaraMeshRendererProperties, rendererIndex=1}
  ]
 }
 Frontend_Kill_Trigger_Big_Tentacles_Bottom001 {
  emitterUpdateScript [
   &EmitterState(Inactive Response=::NewEnumerator1, Life Cycle Mode=::NewEnumerator1, Loop Delay=1.5, Loop Duration=3, UseLoopDelay=true)
   &SpawnBurst_Instantaneous(
    Spawn Count=&MakeCustomIntFromBool(
     Bool=$User.bisFrontEndPreview
     True INT=&MakeCustomIntFromBool(Bool=$User.IsSkydiving, True INT=0, False INT=1)
     False INT=0
    )
    Use Spawn Probability=false
    UseLoopCountLimit=false
   )
  ]
  particleSpawnScript [
   &InitializeParticle(Color Mode=::NewEnumerator0, Lifetime=2.5, Mesh Scale Mode=::NewEnumerator4, UsePositionOffset=false, Write Material Random=true)
   &InitialMeshOrientation(Mesh Orientation Mode=::NewEnumerator2, Rotation=<0 0 -0.25>, Use Rotation Vector=true)
   &InitialMeshOrientation(Mesh Orientation Mode=::NewEnumerator2, Rotation=<0 0.2 0>, Use Rotation Vector=true) {module=InitialMeshOrientation001}
  ]
  particleUpdateScript [
   &CameraOffset(Camera Offset Amount=1)
   &ParticleState()
   &SkeletalMeshLocation(
    Bone Sampling Mode=::NewEnumerator4
    MaskSpawnGroup=false
    Sampled Position Offset=&MakeVector(
     X=0
     Y=&Lerp_Float(
      A=0
      B=-20
      Alpha=&FloatFromCurve(FloatCurve=DataInterface() {propertyValues {Curve {keys [
           {interpMode=RCIM_Cubic, tangentMode=RCTM_User, time=0, value=0}
           <at 0.15 1>
           {interpMode=RCIM_Cubic, tangentMode=RCTM_User, time=0.8, value=1}
           {tangentMode=RCTM_User, time=1, value=0}
          ]}}})
     )
     Z=0
    )
    Skeletal Mesh=DataInterface() {propertyValues {
      DefaultMesh=null
      ExcludeBoneName=root
      OptionalComponentClass=null
      PreviewMesh=&M_MED_Clash
      SkinningMode=SkinOnTheFly
      SoftSourceActor=null
      SourceMode=Default
      UvSetIndex=0
      WholeMeshLOD=-1
      bExcludeBone=false
      bReadDeformedGeometry=true
      bRequireCurrentFrameData=true
      ComponentTags []
      FilteredBones [spine_04]
      FilteredSockets []
      MeshUserParameter {parameter {name=None, typeDefHandle {registeredTypeIndex=1}, varData []}}
      SamplingRegions []
     }}
   ) {module=SkeletalMeshLocation001}
   &DynamicMaterialParameters(
    Index 0 Param 1=0.5
    Index 0 Param 2=1
    Index 0 Param 3=1
    Index 0 Param 4=&FloatFromCurve(FloatCurve=DataInterface() {propertyValues {Curve {keys [
         {interpMode=RCIM_Cubic, tangentMode=RCTM_User, time=0, value=0}
         <at 0.1 0>
         {interpMode=RCIM_Cubic, tangentMode=RCTM_User, time=0.75, value=0}
         {tangentMode=RCTM_User, time=1, value=1}
        ]}}})
    Index 1 Param 3=1
    Index 1 Param 4=0.5
    Module.Param0WriteEnabled=true
    Module.Param0WriteEnabled1=true
    Module.Param1WriteEnabled=true
    Module.Param1WriteEnabled1=true
    Module.Param2WriteEnabled=true
    Module.Param2WriteEnabled1=true
    Module.Param3WriteEnabled=true
    Module.Param3WriteEnabled1=true
    Write Parameter Index 1=true
   )
   &ScaleMeshSize(Scale Factor=&Multiply_VectorByFloat(
     Vector=&Multiply_VectorByFloat(Vector=<1 1 1>, Float=1)
     Float=&Multiply_Float(
      A=&FloatFromCurve(FloatCurve=DataInterface() {propertyValues {Curve {keys [
           {interpMode=RCIM_Cubic, tangentMode=RCTM_User, time=0.03, value=0}
           <at 0.18 1>
           {interpMode=RCIM_Cubic, tangentMode=RCTM_User, time=0.8, value=1}
           {tangentMode=RCTM_User, time=1, value=0}
          ]}}})
      B=1
     )
    )) {module=ScaleMeshSize002}
  ]
  rendererClasses [&NiagaraMeshRendererProperties]
  renderers [{rendererClass=&NiagaraMeshRendererProperties, rendererIndex=0}]
 }
 Frontend_Kill_Trigger_small_Tentacles001 {
  emitterUpdateScript [
   &EmitterState(Inactive Response=::NewEnumerator1, Life Cycle Mode=::NewEnumerator1, Loop Delay=1.5, Loop Duration=3, UseLoopDelay=true)
   &SpawnBurst_Instantaneous(
    Spawn Count=&MakeCustomIntFromBool(
     Bool=$User.bisFrontEndPreview
     True INT=&MakeCustomIntFromBool(Bool=$User.IsSkydiving, True INT=0, False INT=1)
     False INT=0
    )
    Use Spawn Probability=false
    UseLoopCountLimit=false
   )
  ]
  particleSpawnScript [
   &InitializeParticle(
    Color Mode=::NewEnumerator0
    Lifetime=2.5
    Mesh Scale=&Multiply_VectorByFloat(Vector=<1 1 1>, Float=0.7)
    Mesh Scale Mode=::NewEnumerator4
    UsePositionOffset=false
    Write Material Random=true
   )
   &InitialMeshOrientation(Mesh Orientation Mode=::NewEnumerator2, Rotation=<0 0 -0.25>, Use Rotation Vector=true)
   &InitialMeshOrientation(Mesh Orientation Mode=::NewEnumerator2, Rotation=<0 -0.22 0>, Use Rotation Vector=true) {module=InitialMeshOrientation001}
   &InitialMeshOrientation(Mesh Orientation Mode=::NewEnumerator2, Rotation=<-0.01 0 0>, Use Rotation Vector=true) {module=InitialMeshOrientation002}
   &CameraOffset(Camera Offset Amount=2)
  ]
  particleUpdateScript [
   &ParticleState()
   &SkeletalMeshLocation(
    Bone Sampling Mode=::NewEnumerator4
    MaskSpawnGroup=false
    Sampled Position Offset=&MakeVector(
     X=0
     Y=&Lerp_Float(
      A=0
      B=-15
      Alpha=&FloatFromCurve(FloatCurve=DataInterface() {propertyValues {Curve {keys [
           {interpMode=RCIM_Cubic, tangentMode=RCTM_User, time=0.1, value=0}
           <at 0.25 1>
           {interpMode=RCIM_Cubic, tangentMode=RCTM_User, time=0.8, value=1}
           {tangentMode=RCTM_User, time=1, value=0}
          ]}}})
     )
     Z=0
    )
    Skeletal Mesh=DataInterface() {propertyValues {
      DefaultMesh=null
      ExcludeBoneName=root
      OptionalComponentClass=null
      PreviewMesh=&M_MED_Clash
      SkinningMode=SkinOnTheFly
      SoftSourceActor=null
      SourceMode=Default
      UvSetIndex=0
      WholeMeshLOD=-1
      bExcludeBone=false
      bReadDeformedGeometry=true
      bRequireCurrentFrameData=true
      ComponentTags []
      FilteredBones [spine_04]
      FilteredSockets []
      MeshUserParameter {parameter {name=None, typeDefHandle {registeredTypeIndex=1}, varData []}}
      SamplingRegions []
     }}
   )
   &DynamicMaterialParameters(
    Index 0 Param 2=1.2
    Index 0 Param 3=1.5
    Index 0 Param 4=&FloatFromCurve(FloatCurve=DataInterface() {propertyValues {Curve {keys [
         {interpMode=RCIM_Cubic, tangentMode=RCTM_User, time=0, value=0}
         <at 0.1 0>
         {interpMode=RCIM_Cubic, tangentMode=RCTM_User, time=0.75, value=0}
         {tangentMode=RCTM_User, time=1, value=1}
        ]}}})
    Index 1 Param 3=1
    Index 1 Param 4=0.5
    Module.Param0WriteEnabled=true
    Module.Param0WriteEnabled1=true
    Module.Param1WriteEnabled=true
    Module.Param1WriteEnabled1=true
    Module.Param2WriteEnabled=true
    Module.Param2WriteEnabled1=true
    Module.Param3WriteEnabled=true
    Module.Param3WriteEnabled1=true
    Write Parameter Index 1=true
   )
   &ScaleMeshSize(Scale Factor=&Multiply_VectorByFloat(
     Vector=&Multiply_VectorByFloat(Vector=<1 1 1>, Float=1)
     Float=&FloatFromCurve(FloatCurve=DataInterface() {propertyValues {Curve {keys [
          {interpMode=RCIM_Cubic, tangentMode=RCTM_User, time=0.1, value=0}
          <at 0.25 1>
          {interpMode=RCIM_Cubic, tangentMode=RCTM_User, time=0.8, value=1}
          {tangentMode=RCTM_User, time=1, value=0}
         ]}}})
    )) {module=ScaleMeshSize001}
  ]
  rendererClasses [&NiagaraMeshRendererProperties]
  renderers [{rendererClass=&NiagaraMeshRendererProperties, rendererIndex=0}]
 }
 Frontend_Small_MeshTendrils001 {
  emitterUpdateScript [
   &EmitterState(Inactive Response=::NewEnumerator1, Life Cycle Mode=::NewEnumerator1, Loop Delay=1.5, Loop Duration=3, UseLoopDelay=true)
   &SpawnRate(
    SpawnRate=&Multiply_Float(
     A=&FloatFromCurve(
      FloatCurve=DataInterface() {propertyValues {Curve {keys [<at 0 0>,<at 0.1 80>,<at 0.3 80>,<at 0.5 0>]}}}
      CurveIndex=$Emitter.NormalizedLoopAge
     )
     B=&MakeCustomFloatFromBool(
      Bool=$User.bisFrontEndPreview
      True Float=&MakeCustomFloatFromBool(Bool=$User.IsSkydiving, True Float=0, False Float=1)
      False Float=0
     )
    )
    Use Spawn Probability=false
   )
  ]
  particleSpawnScript [
   &InitializeParticle(
    Color=<rgba 0.276 0.276 0.276 1>
    Lifetime Mode=::NewEnumerator1
    Mesh Scale=&Multiply_VectorByFloat(
     Vector=&RandomRangeVector(Minimum=<0.7 0.7 0.8>, Maximum=<1 0.8 1>)
     Float=&RandomRangeFloat(Minimum=0.9, Maximum=1.2)
    )
    Mesh Scale Mode=::NewEnumerator4
    UsePositionOffset=false
    Write Material Random=true
   ) {module=InitializeParticle001}
   &FN_SetSpawnGroup(Number of Spawn Groups=14)
   &InitialMeshOrientation(
    Rotation=&RandomRangeVector(Minimum=<-0.14 -0.07 0>, Maximum=<0.14 0.07 0>)
    Use Orientation Vector=true
    Use Rotation Vector=true
   ) {module=InitialMeshOrientation001}
   /SetVariables(Particles.VisibilityTag=&RandomRangeInt(Minimum=0, Maximum=1)) {isSetParameters=true, module=SetVariables_#4}
   &CameraOffset(Camera Offset Amount=&RandomRangeFloat(Minimum=-10, Maximum=-5))
  ]
  particleUpdateScript [
   /SetVariables(Particles.Mesh_Scale02:Vector3f=&Multiply_Vector(
     A=&Multiply_Vector(
      A=&Multiply_VectorByFloat(
       Vector=&RandomRangeVector(Minimum=<0.6 0.6 0.8>, Maximum=<0.8 0.8 1>)
       Float=&RandomRangeFloat(Minimum=0.9, Maximum=1.4)
      )
      B=&VectorFromCurve(
       VectorCurve=DataInterface() {propertyValues {
         XCurve {keys [<at 0 0 RCIM_Cubic>,<at 0.5 1 RCIM_Cubic>,<at 1 0 RCIM_Cubic>]}
         YCurve {keys [<at 0 0 RCIM_Cubic>,<at 0.5 1 RCIM_Cubic>,<at 1 0 RCIM_Cubic>]}
         ZCurve {keys [
           {arriveTangent=2.4741, interpMode=RCIM_Cubic, leaveTangent=2.4741, tangentMode=RCTM_User, time=0, value=0}
           <at 0.5 1 RCIM_Cubic>
           {arriveTangent=-2.7109, interpMode=RCIM_Cubic, leaveTangent=-2.7109, tangentMode=RCTM_User, time=1, value=0}
          ]}
        }}
       Scale Curve=&VectorFromFloat(Value=0.85)
      )
     )
     B=&VectorFromFloat(Value=&MakeCustomFloatFromBool(Bool=$User.Is_SkinFX_Trigger, True Float=1, False Float=0))
    )) {isSetParameters=true, module=SetVariables_#5}
   &ParticleState()
   &DynamicMaterialParameters(
    Index 0 Param 1=&FloatFromCurve(FloatCurve=DataInterface() {propertyValues {Curve {keys [<at 0 0 RCIM_Cubic>,<at 0.5 1 RCIM_Cubic>,<at 1 0 RCIM_Cubic>]}}}, Scale Curve=1.25)
    Index 0 Param 2=&FloatFromCurve(
     FloatCurve=DataInterface() {propertyValues {Curve {keys [<at 0 1 RCIM_Cubic>,<at 0.5 0.1 RCIM_Cubic>,<at 1 1 RCIM_Cubic>]}}}
     Scale Curve=3.5
    )
    Index 0 Param 4=1
    Index 1 Param 3=&MakeCustomFloatFromBool(Bool=$User.bisFrontEndPreview, True Float=1, False Float=1)
    Module.Param0WriteEnabled=true
    Module.Param0WriteEnabled1=true
    Module.Param1WriteEnabled=true
    Module.Param1WriteEnabled1=true
    Module.Param2WriteEnabled=true
    Module.Param2WriteEnabled1=true
    Module.Param3WriteEnabled=true
    Module.Param3WriteEnabled1=true
    Write Parameter Index 1=true
   )
   &SkeletalMeshLocation(
    Apply Sampled Orientation to Mesh Particles=true
    Bone / Socket Index=$Particles.SpawnGroup
    Inherit Velocity=false
    MaskSpawnGroup=false
    Mesh Sampling Type=::NewEnumerator5
    Sampled Position Offset=&MakeVector(
     X=&RandomRangeFloat(Minimum=0, Maximum=10)
     Y=0
     Z=&RandomRangeFloat(Minimum=-2, Maximum=-2)
    )
    Sampled Position Offset Coordinate Space:ENiagaraMeshTransforms=::NewEnumerator3
    Skeletal Mesh=$User.SK_Mesh_Tendrils
    Socket Sampling Mode=::NewEnumerator5
    Use Sampled Mesh Attribute as Ribbon ID=false
   )
   &ScaleMeshSize(Scale Factor=&Multiply_Vector(
     A=&VectorFromCurve(
      VectorCurve=DataInterface() {propertyValues {
        XCurve {keys [<at 0 0 RCIM_Cubic>,<at 0.5 1 RCIM_Cubic>,<at 1 0 RCIM_Cubic>]}
        YCurve {keys [<at 0 0 RCIM_Cubic>,<at 0.5 1 RCIM_Cubic>,<at 1 0 RCIM_Cubic>]}
        ZCurve {keys [
          {arriveTangent=2.4741, interpMode=RCIM_Cubic, leaveTangent=2.4741, tangentMode=RCTM_User, time=0, value=0}
          <at 0.5 1 RCIM_Cubic>
          {arriveTangent=-2.7109, interpMode=RCIM_Cubic, leaveTangent=-2.7109, tangentMode=RCTM_User, time=1, value=0}
         ]}
       }}
      Scale Curve=&VectorFromFloat(Value=0.85)
     )
     B=<1 1 1>
    )) {module=ScaleMeshSize001}
  ]
  rendererClasses [&NiagaraMeshRendererProperties]
  renderers [
   {rendererClass=&NiagaraMeshRendererProperties, rendererIndex=0}
   {rendererClass=&NiagaraMeshRendererProperties, rendererIndex=1}
  ]
 }
 Frontend_Small_MeshTendrils_burst {
  emitterUpdateScript [
   &EmitterState(Inactive Response=::NewEnumerator1, Life Cycle Mode=::NewEnumerator1, Loop Delay=1.5, Loop Duration=3, UseLoopDelay=true)
   &SpawnBurst_Instantaneous(
    Spawn Count=&MakeCustomIntFromBool(
     Bool=$User.bisFrontEndPreview
     True INT=&MakeCustomIntFromBool(Bool=$User.IsSkydiving, True INT=0, False INT=40)
     False INT=0
    )
    Use Spawn Probability=false
    UseLoopCountLimit=false
   )
  ]
  particleSpawnScript [
   &InitializeParticle(
    Color=<rgba 0.276 0.276 0.276 1>
    Lifetime=1.5
    Mesh Scale=&Multiply_VectorByFloat(
     Vector=&RandomRangeVector(Minimum=<0.7 0.7 1.1>, Maximum=<1 0.8 1.1>)
     Float=&RandomRangeFloat(Minimum=0.9, Maximum=1.1)
    )
    Mesh Scale Mode=::NewEnumerator4
    UsePositionOffset=false
    Write Material Random=true
   ) {module=InitializeParticle001}
   &FN_SetSpawnGroup(Number of Spawn Groups=14)
   &InitialMeshOrientation(
    Rotation=&RandomRangeVector(Minimum=<-0.14 -0.07 0>, Maximum=<0.14 0.07 0>)
    Use Orientation Vector=true
    Use Rotation Vector=true
   ) {module=InitialMeshOrientation001}
   /SetVariables(Particles.VisibilityTag=&RandomRangeInt(Minimum=0, Maximum=1)) {isSetParameters=true, module=SetVariables_#4}
   &CameraOffset(Camera Offset Amount=&RandomRangeFloat(Minimum=-10, Maximum=-5))
  ]
  particleUpdateScript [
   &ParticleState()
   &DynamicMaterialParameters(
    Index 0 Param 1=&FloatFromCurve(FloatCurve=DataInterface() {propertyValues {Curve {keys [<at 0 0 RCIM_Cubic>,<at 0.5 1 RCIM_Cubic>,<at 1 0 RCIM_Cubic>]}}}, Scale Curve=1.25)
    Index 0 Param 2=&FloatFromCurve(
     FloatCurve=DataInterface() {propertyValues {Curve {keys [<at 0 1 RCIM_Cubic>,<at 0.5 0.1 RCIM_Cubic>,<at 1 1 RCIM_Cubic>]}}}
     Scale Curve=3.5
    )
    Index 0 Param 4=1
    Module.Param0WriteEnabled=true
    Module.Param1WriteEnabled=true
    Module.Param2WriteEnabled=true
    Module.Param3WriteEnabled=true
   )
   &SkeletalMeshLocation(
    Apply Sampled Orientation to Mesh Particles=true
    Bone / Socket Index=$Particles.SpawnGroup
    Inherit Velocity=false
    MaskSpawnGroup=false
    Mesh Sampling Type=::NewEnumerator5
    Sampled Position Offset=&MakeVector(
     X=&RandomRangeFloat(Minimum=0, Maximum=10)
     Y=0
     Z=&RandomRangeFloat(Minimum=-2, Maximum=-2)
    )
    Sampled Position Offset Coordinate Space:ENiagaraMeshTransforms=::NewEnumerator3
    Skeletal Mesh=$User.SK_Mesh_Tendrils
    Socket Sampling Mode=::NewEnumerator5
    Use Sampled Mesh Attribute as Ribbon ID=false
   )
   &ScaleMeshSize(Scale Factor=&Multiply_Vector(
     A=<1 1 1>
     B=&Multiply_VectorByFloat(
      Vector=<1 1 1>
      Float=&FloatFromCurve(FloatCurve=DataInterface() {propertyValues {Curve {keys [
           {interpMode=RCIM_Cubic, tangentMode=RCTM_User, time=0.1, value=0}
           {interpMode=RCIM_Cubic, leaveTangent=-2.2097, tangentMode=RCTM_Break, time=0.4, value=1}
           {arriveTangent=-1.293, interpMode=RCIM_Cubic, leaveTangent=-8.0089, tangentMode=RCTM_Break, time=0.9, value=0}
          ]}}})
     )
    )) {module=ScaleMeshSize001}
  ]
  rendererClasses [&NiagaraMeshRendererProperties]
  renderers [{rendererClass=&NiagaraMeshRendererProperties, rendererIndex=0}]
 }
 Kill_Trigger_Big_Tentacles {
  emitterUpdateScript [
   &EmitterState(Life Cycle Mode=::NewEnumerator1, Loop Delay=0.5, Loop Duration=3, UseLoopDelay=false)
   &SpawnBurst_Instantaneous(
    Spawn Count=&MakeCustomIntFromBool(Bool=$User.bisInVehicle, True INT=0, False INT=1)
    Use Spawn Probability=false
    UseLoopCountLimit=false
   )
  ]
  particleSpawnScript [
   &InitializeParticle(Color Mode=::NewEnumerator0, Lifetime=3, Mesh Scale Mode=::NewEnumerator4, UsePositionOffset=false, Write Material Random=true)
   &InitialMeshOrientation(Mesh Orientation Mode=::NewEnumerator2, Rotation=<0 0 -0.25>, Use Rotation Vector=true)
   &InitialMeshOrientation(Mesh Orientation Mode=::NewEnumerator2, Rotation=<0 0.2 0>, Use Rotation Vector=true) {module=InitialMeshOrientation001}
   &CameraOffset()
  ]
  particleUpdateScript [
   &ParticleState()
   &SkeletalMeshLocation(
    Bone Sampling Mode=::NewEnumerator4
    MaskSpawnGroup=false
    Sampled Position Offset=&MakeVector(X=0, Y=&Lerp_Float(A=0, B=-20, Alpha=$User.Big_MeshTendrils_Location), Z=0)
    Skeletal Mesh=DataInterface() {propertyValues {
      DefaultMesh=null
      ExcludeBoneName=root
      OptionalComponentClass=null
      PreviewMesh=&M_MED_Clash
      SkinningMode=SkinOnTheFly
      SoftSourceActor=null
      SourceMode=Default
      UvSetIndex=0
      WholeMeshLOD=-1
      bExcludeBone=false
      bReadDeformedGeometry=true
      bRequireCurrentFrameData=true
      ComponentTags []
      FilteredBones [spine_04]
      FilteredSockets []
      MeshUserParameter {parameter {name=None, typeDefHandle {registeredTypeIndex=1}, varData []}}
      SamplingRegions []
     }}
   )
   &DynamicMaterialParameters(
    Index 0 Param 1=0.5
    Index 0 Param 2=1.5
    Index 0 Param 3=1.5
    Index 0 Param 4=$User.Big_MeshTendrils_Dissolve
    Index 1 Param 3=1
    Index 1 Param 4=0.5
    Module.Param0WriteEnabled=true
    Module.Param0WriteEnabled1=true
    Module.Param1WriteEnabled=true
    Module.Param1WriteEnabled1=true
    Module.Param2WriteEnabled=true
    Module.Param2WriteEnabled1=true
    Module.Param3WriteEnabled=true
    Module.Param3WriteEnabled1=true
    Write Parameter Index 1=true
   )
   &ScaleMeshSize(Scale Factor=&Multiply_VectorByFloat(Vector=&Multiply_VectorByFloat(Vector=<1 1 1>, Float=1), Float=$User.Big_MeshTendrils_Size)) {module=ScaleMeshSize001}
   /SetVariables(Particles.Mesh02_Size:Vector3f=&Multiply_Vector(A=<1 1 1>, B=&VectorFromFloat(Value=$User.Big_Mesh02Tendrils_Size))) {isSetParameters=true, module=SetVariables_#3}
  ]
  rendererClasses [&NiagaraMeshRendererProperties]
  renderers [
   {rendererClass=&NiagaraMeshRendererProperties, rendererIndex=0}
   {rendererClass=&NiagaraMeshRendererProperties, rendererIndex=1}
  ]
 }
 Kill_Trigger_Big_Tentacles_Bottom {
  emitterUpdateScript [
   &EmitterState(Life Cycle Mode=::NewEnumerator1, Loop Delay=0.5, Loop Duration=3, UseLoopDelay=false)
   &SpawnBurst_Instantaneous(
    Spawn Count=&MakeCustomIntFromBool(Bool=$User.bisInVehicle, True INT=0, False INT=1)
    Use Spawn Probability=false
    UseLoopCountLimit=false
   )
  ]
  particleSpawnScript [
   &InitializeParticle(Color Mode=::NewEnumerator0, Lifetime=3, Mesh Scale Mode=::NewEnumerator4, UsePositionOffset=false, Write Material Random=true)
   &InitialMeshOrientation(Mesh Orientation Mode=::NewEnumerator2, Rotation=<0 0 -0.25>, Use Rotation Vector=true)
   &InitialMeshOrientation(Mesh Orientation Mode=::NewEnumerator2, Rotation=<0 0.2 0>, Use Rotation Vector=true) {module=InitialMeshOrientation001}
  ]
  particleUpdateScript [
   &CameraOffset(Camera Offset Amount=1)
   &ParticleState()
   &SkeletalMeshLocation(
    Bone Sampling Mode=::NewEnumerator4
    MaskSpawnGroup=false
    Sampled Position Offset=&MakeVector(X=0, Y=&Lerp_Float(A=0, B=-20, Alpha=$User.Big_MeshTendrils_Location), Z=0)
    Skeletal Mesh=DataInterface() {propertyValues {
      DefaultMesh=null
      ExcludeBoneName=root
      OptionalComponentClass=null
      PreviewMesh=&M_MED_Clash
      SkinningMode=SkinOnTheFly
      SoftSourceActor=null
      SourceMode=Default
      UvSetIndex=0
      WholeMeshLOD=-1
      bExcludeBone=false
      bReadDeformedGeometry=true
      bRequireCurrentFrameData=true
      ComponentTags []
      FilteredBones [spine_04]
      FilteredSockets []
      MeshUserParameter {parameter {name=None, typeDefHandle {registeredTypeIndex=1}, varData []}}
      SamplingRegions []
     }}
   )
   &DynamicMaterialParameters(
    Index 0 Param 1=0.5
    Index 0 Param 2=1
    Index 0 Param 3=1
    Index 1 Param 3=1
    Index 1 Param 4=0.5
    Module.Param0WriteEnabled=true
    Module.Param0WriteEnabled1=true
    Module.Param1WriteEnabled=true
    Module.Param1WriteEnabled1=true
    Module.Param2WriteEnabled=true
    Module.Param2WriteEnabled1=true
    Module.Param3WriteEnabled=true
    Module.Param3WriteEnabled1=true
    Write Parameter Index 1=true
   )
   &ScaleMeshSize(Scale Factor=&Multiply_VectorByFloat(
     Vector=&Multiply_VectorByFloat(Vector=<1 1 1>, Float=1)
     Float=&Multiply_Float(A=$User.Big_MeshTendrils_Size, B=1)
    )) {module=ScaleMeshSize002}
  ]
  rendererClasses [&NiagaraMeshRendererProperties]
  renderers [{rendererClass=&NiagaraMeshRendererProperties, rendererIndex=0}]
 }
 Kill_Trigger_small_Tentacles {
  emitterUpdateScript [
   &EmitterState(Life Cycle Mode=::NewEnumerator1, Loop Delay=0.2, Loop Duration=3, UseLoopDelay=false)
   &SpawnBurst_Instantaneous(
    Spawn Count=&MakeCustomIntFromBool(Bool=$User.bisInVehicle, True INT=0, False INT=1)
    Use Spawn Probability=false
    UseLoopCountLimit=false
   )
  ]
  particleSpawnScript [
   &InitializeParticle(
    Color Mode=::NewEnumerator0
    Lifetime=3
    Mesh Scale=&Multiply_VectorByFloat(Vector=<1 1 1>, Float=0.7)
    Mesh Scale Mode=::NewEnumerator4
    UsePositionOffset=false
    Write Material Random=true
   )
   &InitialMeshOrientation(Mesh Orientation Mode=::NewEnumerator2, Rotation=<0 0 -0.25>, Use Rotation Vector=true)
   &InitialMeshOrientation(Mesh Orientation Mode=::NewEnumerator2, Rotation=<0 -0.22 0>, Use Rotation Vector=true) {module=InitialMeshOrientation001}
   &CameraOffset(Camera Offset Amount=5)
  ]
  particleUpdateScript [
   &ParticleState()
   &SkeletalMeshLocation(
    Bone Sampling Mode=::NewEnumerator4
    MaskSpawnGroup=false
    Sampled Position Offset=&MakeVector(X=0, Y=&Lerp_Float(A=0, B=-15, Alpha=$User.Small_MeshTendrils_Location), Z=0)
    Skeletal Mesh=DataInterface() {propertyValues {
      DefaultMesh=null
      ExcludeBoneName=root
      OptionalComponentClass=null
      PreviewMesh=&M_MED_Clash
      SkinningMode=SkinOnTheFly
      SoftSourceActor=null
      SourceMode=Default
      UvSetIndex=0
      WholeMeshLOD=-1
      bExcludeBone=false
      bReadDeformedGeometry=true
      bRequireCurrentFrameData=true
      ComponentTags []
      FilteredBones [spine_04]
      FilteredSockets []
      MeshUserParameter {parameter {name=None, typeDefHandle {registeredTypeIndex=1}, varData []}}
      SamplingRegions []
     }}
   )
   &DynamicMaterialParameters(
    Index 0 Param 2=1.5
    Index 0 Param 3=1.5
    Index 0 Param 4=$User.Big_MeshTendrils_Dissolve
    Index 1 Param 3=1
    Index 1 Param 4=0.5
    Module.Param0WriteEnabled=true
    Module.Param0WriteEnabled1=true
    Module.Param1WriteEnabled=true
    Module.Param1WriteEnabled1=true
    Module.Param2WriteEnabled=true
    Module.Param2WriteEnabled1=true
    Module.Param3WriteEnabled=true
    Module.Param3WriteEnabled1=true
    Write Parameter Index 1=true
   )
   &ScaleMeshSize(Scale Factor=&Multiply_VectorByFloat(Vector=&Multiply_VectorByFloat(Vector=<1 1 1>, Float=1), Float=$User.Small_MeshTendrils_Size)) {module=ScaleMeshSize001}
  ]
  rendererClasses [&NiagaraMeshRendererProperties]
  renderers [{rendererClass=&NiagaraMeshRendererProperties, rendererIndex=0}]
 }
 Small_MeshTendrils {
  emitterUpdateScript [
   &EmitterState(Life Cycle Mode=::NewEnumerator1, Loop Duration=1, UseLoopDelay=false)
   &SpawnRate(SpawnRate=$User.Spikes_number, Use Spawn Probability=false)
  ]
  particleSpawnScript [
   &InitializeParticle(
    Color=<rgba 0.276 0.276 0.276 1>
    Lifetime Mode=::NewEnumerator1
    Mesh Scale=&Multiply_VectorByFloat(
     Vector=&RandomRangeVector(Minimum=<0.7 0.7 0.8>, Maximum=<1 0.8 1>)
     Float=&RandomRangeFloat(Minimum=0.9, Maximum=1.2)
    )
    Mesh Scale Mode=::NewEnumerator4
    UsePositionOffset=false
    Write Material Random=true
   ) {module=InitializeParticle001}
   &FN_SetSpawnGroup(Number of Spawn Groups=14)
   &InitialMeshOrientation(
    Rotation=&RandomRangeVector(Minimum=<-0.14 -0.07 0>, Maximum=<0.14 0.07 0>)
    Use Orientation Vector=true
    Use Rotation Vector=true
   ) {module=InitialMeshOrientation001}
   /SetVariables(Particles.VisibilityTag=&RandomRangeInt(Minimum=0, Maximum=1)) {isSetParameters=true, module=SetVariables_#4}
   &CameraOffset(Camera Offset Amount=&RandomRangeFloat(Minimum=-10, Maximum=-5))
  ]
  particleUpdateScript [
   /SetVariables(Particles.Mesh_Scale02:Vector3f=&Multiply_Vector(
     A=&Multiply_Vector(
      A=&Multiply_VectorByFloat(
       Vector=&RandomRangeVector(Minimum=<0.6 0.6 0.8>, Maximum=<0.8 0.8 1>)
       Float=&RandomRangeFloat(Minimum=0.9, Maximum=1.4)
      )
      B=&VectorFromCurve(
       VectorCurve=DataInterface() {propertyValues {
         XCurve {keys [<at 0 0 RCIM_Cubic>,<at 0.5 1 RCIM_Cubic>,<at 1 0 RCIM_Cubic>]}
         YCurve {keys [<at 0 0 RCIM_Cubic>,<at 0.5 1 RCIM_Cubic>,<at 1 0 RCIM_Cubic>]}
         ZCurve {keys [
           {arriveTangent=2.4741, interpMode=RCIM_Cubic, leaveTangent=2.4741, tangentMode=RCTM_User, time=0, value=0}
           <at 0.5 1 RCIM_Cubic>
           {arriveTangent=-2.7109, interpMode=RCIM_Cubic, leaveTangent=-2.7109, tangentMode=RCTM_User, time=1, value=0}
          ]}
        }}
       Scale Curve=&VectorFromFloat(Value=0.85)
      )
     )
     B=&VectorFromFloat(Value=&MakeCustomFloatFromBool(Bool=$User.Is_SkinFX_Trigger, True Float=1, False Float=0))
    )) {isSetParameters=true, module=SetVariables_#5}
   &ParticleState()
   &DynamicMaterialParameters(
    Index 0 Param 1=&FloatFromCurve(FloatCurve=DataInterface() {propertyValues {Curve {keys [<at 0 0 RCIM_Cubic>,<at 0.5 1 RCIM_Cubic>,<at 1 0 RCIM_Cubic>]}}}, Scale Curve=1.25)
    Index 0 Param 2=&FloatFromCurve(
     FloatCurve=DataInterface() {propertyValues {Curve {keys [<at 0 1 RCIM_Cubic>,<at 0.5 0.1 RCIM_Cubic>,<at 1 1 RCIM_Cubic>]}}}
     Scale Curve=3.5
    )
    Index 0 Param 4=1
    Index 1 Param 3=&MakeCustomFloatFromBool(Bool=$User.bisFrontEndPreview, True Float=1, False Float=1)
    Module.Param0WriteEnabled=true
    Module.Param0WriteEnabled1=true
    Module.Param1WriteEnabled=true
    Module.Param1WriteEnabled1=true
    Module.Param2WriteEnabled=true
    Module.Param2WriteEnabled1=true
    Module.Param3WriteEnabled=true
    Module.Param3WriteEnabled1=true
    Write Parameter Index 1=true
   )
   &SkeletalMeshLocation(
    Apply Sampled Orientation to Mesh Particles=true
    Bone / Socket Index=$Particles.SpawnGroup
    Inherit Velocity=false
    MaskSpawnGroup=false
    Mesh Sampling Type=::NewEnumerator5
    Sampled Position Offset=&MakeVector(
     X=&RandomRangeFloat(Minimum=0, Maximum=10)
     Y=0
     Z=&RandomRangeFloat(Minimum=-2, Maximum=-2)
    )
    Sampled Position Offset Coordinate Space:ENiagaraMeshTransforms=::NewEnumerator3
    Skeletal Mesh=$User.SK_Mesh_Tendrils
    Socket Sampling Mode=::NewEnumerator5
    Use Sampled Mesh Attribute as Ribbon ID=false
   )
   &ScaleMeshSize(Scale Factor=&Multiply_Vector(
     A=&VectorFromCurve(
      VectorCurve=DataInterface() {propertyValues {
        XCurve {keys [<at 0 0 RCIM_Cubic>,<at 0.5 1 RCIM_Cubic>,<at 1 0 RCIM_Cubic>]}
        YCurve {keys [<at 0 0 RCIM_Cubic>,<at 0.5 1 RCIM_Cubic>,<at 1 0 RCIM_Cubic>]}
        ZCurve {keys [
          {arriveTangent=2.4741, interpMode=RCIM_Cubic, leaveTangent=2.4741, tangentMode=RCTM_User, time=0, value=0}
          <at 0.5 1 RCIM_Cubic>
          {arriveTangent=-2.7109, interpMode=RCIM_Cubic, leaveTangent=-2.7109, tangentMode=RCTM_User, time=1, value=0}
         ]}
       }}
      Scale Curve=&VectorFromFloat(Value=0.85)
     )
     B=<1 1 1>
    )) {module=ScaleMeshSize001}
   /SetVariables(Particles.VisibilityTag=&self:PartHiddenBoneCheck(isHandHidden:NiagaraBool=$User.bisHandHidden, isFootHidden:NiagaraBool=$User.bisFootHidden, Selection:NiagaraInt32=$Particles.SpawnGroup)) {isSetParameters=true}
  ]
  rendererClasses [&NiagaraMeshRendererProperties]
  renderers [
   {rendererClass=&NiagaraMeshRendererProperties, rendererIndex=0}
   {rendererClass=&NiagaraMeshRendererProperties, rendererIndex=1}
  ]
 }
}
summary {
 systemName=NS_MED_Clash_Skin_Red
 emitters [
  {emitterName=Bone_Flow_Ribbon01, rendererClasses [&NiagaraRibbonRendererProperties]}
  {emitterName=Small_MeshTendrils, rendererClasses [&NiagaraMeshRendererProperties]}
  {emitterName=Bone_Flow_Ribbon02, rendererClasses [&NiagaraRibbonRendererProperties]}
  {emitterName=FN_AttachedRibbon_v3003, rendererClasses [&NiagaraRibbonRendererProperties]}
  {emitterName=Kill_Trigger_Big_Tentacles_Bottom, rendererClasses [&NiagaraMeshRendererProperties]}
  {emitterName=Kill_Trigger_Big_Tentacles, rendererClasses [&NiagaraMeshRendererProperties]}
  {emitterName=Kill_Trigger_small_Tentacles, rendererClasses [&NiagaraMeshRendererProperties]}
  {emitterName=Frontend_Kill_Trigger_Big_Tentacles_Bottom001, rendererClasses [&NiagaraMeshRendererProperties]}
  {emitterName=Frontend_Kill_Trigger_Big_Tentacles001, rendererClasses [&NiagaraMeshRendererProperties]}
  {emitterName=Frontend_Kill_Trigger_small_Tentacles001, rendererClasses [&NiagaraMeshRendererProperties]}
  {emitterName=Frontend_Small_MeshTendrils001, rendererClasses [&NiagaraMeshRendererProperties]}
  {emitterName=Frontend_Small_MeshTendrils_burst, rendererClasses [&NiagaraMeshRendererProperties]}
 ]
 userVariables [
  User.Big_Mesh02Tendrils_Size {type=NiagaraFloat, defaultValue=0, description=""}
  User.Big_MeshTendrils_Dissolve {type=NiagaraFloat, defaultValue=0, description=""}
  User.Big_MeshTendrils_Location {type=NiagaraFloat, defaultValue=0, description=""}
  User.Big_MeshTendrils_Size {type=NiagaraFloat, defaultValue=0, description=""}
  User.bisFootHidden {type=NiagaraBool, defaultValue=true}
  User.bisFrontEndPreview {type=NiagaraBool}
  User.bisHandHidden {type=NiagaraBool}
  User.bisInVehicle {type=NiagaraBool}
  User.Is_SkinFX_Trigger {type=NiagaraBool}
  User.IsSkydiving {type=NiagaraBool}
  User.Kill {type=NiagaraBool}
  User.RibbonIDOffset {type=NiagaraInt32}
  User.SK_Mesh {
   type=NiagaraDataInterfaceSkeletalMesh
   defaultValue {
    struct=NiagaraExt_VariableValue_DataInterface
    value {dataInterface=&self:NiagaraDataInterfaceSkeletalMesh_0}
   }
  }
  User.SK_Mesh_Tendrils {
   type=NiagaraDataInterfaceSkeletalMesh
   defaultValue {
    struct=NiagaraExt_VariableValue_DataInterface
    value {dataInterface=&self:NiagaraDataInterfaceSkeletalMesh_2}
   }
  }
  User.SkeletalMesh {
   type=NiagaraDataInterfaceSkeletalMesh
   defaultValue {
    struct=NiagaraExt_VariableValue_DataInterface
    value {dataInterface=&self:NiagaraDataInterfaceSkeletalMesh_3}
   }
  }
  User.SkeletalMesh_Trail_Scoket {
   type=NiagaraDataInterfaceSkeletalMesh
   defaultValue {
    struct=NiagaraExt_VariableValue_DataInterface
    value {dataInterface=&self:NiagaraDataInterfaceSkeletalMesh_6}
   }
  }
  User.Small_MeshTendrils_Burst {type=NiagaraInt32}
  User.Small_MeshTendrils_Location {type=NiagaraFloat, defaultValue=0, description=""}
  User.Small_MeshTendrils_Size {type=NiagaraFloat, defaultValue=0, description=""}
  User.Spikes_number {type=NiagaraFloat, defaultValue=20, description=""}
 ]
}
