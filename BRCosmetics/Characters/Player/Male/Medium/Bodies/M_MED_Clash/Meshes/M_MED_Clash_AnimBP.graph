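# Post-copy AnimBlueprint for the Clash male body cosmetic, adding deformation and shrinking the right hand and forearm twists for the Clash pickaxe.
# The pose copied from the attached parent runs through twist and sculpter Control Rigs, then bone modifiers, including fixed upper-arm and tricep offsets.
# Each update sets the shrink alpha from weapon visibility and the pickaxe's Clash meta tag.
# Parachute events set HangGliding, which the AnimGraph never reads.
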AnimBlueprint /BRCosmetics/Characters/Player/Male/Medium/Bodies/M_MED_Clash/Meshes/M_MED_Clash_AnimBP
target_skeleton=&/M_MED_Clash_Skeleton
graphs [
 AnimGraph {nodes [
   n1 = Control Rig(Source=Copy Pose From Mesh() {bUseAttachedParent=true}) {
    alphaBoolBlend {blendOption=Linear}
    controlRigAssetReference {blueprintRigClass=&/M_MED_Clash_Deform_CtrlRig_twists.M_MED_Clash_Deform_CtrlRig_twists_C}
   }
   n2 = Control Rig!(
    Source=@n1
    Alpha
    KneeRot_fr_l_114deg:Vector
    KneeTrans_fr_l_114deg:Vector="0.258,-3.591,0.009"
    KneeScale_fr_l_114deg:Vector="1,1,1"
    KneeScale_fr_l_130deg:Vector="1,1,1"
    KneeRot_fr_l_130deg:Vector="0,0,-51.2"
    KneeTrans_fr_l_130deg:Vector="0.258,-3.591,0.009"
   ) -> (Pose) {
    alphaBoolBlend {blendOption=Linear}
    controlRigAssetReference {blueprintRigClass=&/M_MED_Clash_Deform_CtrlRig_sculpters.M_MED_Clash_Deform_CtrlRig_sculpters_C}
   }
   n3 = Get ClashPickaxe() -> (ClashPickaxe:double)
   n4 = "Transform (Modify) Bone - Bone: hand_r"(ComponentPose=Local To Component(LocalPose=@n2), Alpha=@n3, Translation="0,0,0", Rotation="0,0,0", Scale="0.1,0.1,0.1") {scaleMode=BMM_Replace, scaleSpace=BCS_BoneSpace, alphaBoolBlend {blendOption=Linear}, boneToModify {boneName=hand_r}, scale=<0.1 0.1 0.1>}
   n5 = "Transform (Modify) Bone - Bone: lowerarm_twist_02_r"(ComponentPose=@n4, Alpha=@n3, Translation="0,0,0", Rotation="0,0,0", Scale="0.75,0.75,0.75") {
    scaleMode=BMM_Replace
    scaleSpace=BCS_BoneSpace
    alphaBoolBlend {blendOption=Linear}
    boneToModify {boneName=lowerarm_twist_02_r}
    scale=<0.75 0.75 0.75>
   }
   n6 = "Transform (Modify) Bone - Bone: lowerarm_twist_01_r"(ComponentPose=@n5, Alpha=@n3, Translation="0,0,0", Rotation="0,0,0", Scale="0.75,0.75,0.75") {
    scaleMode=BMM_Replace
    scaleSpace=BCS_BoneSpace
    alphaBoolBlend {blendOption=Linear}
    alphaScaleBiasClamp {bInterpResult=true}
    boneToModify {boneName=lowerarm_twist_01_r}
    scale=<0.75 0.75 0.75>
   }
   n7 = "Transform (Modify) Bone - Bone: upperarm_twist_01_r"!(ComponentPose=@n6, Alpha=1, Rotation="-12,0,0") -> (Pose) {
    rotationMode=BMM_Additive
    alphaBoolBlend {blendOption=Linear}
    alphaScaleBiasClamp {bInterpResult=true, interpSpeedDecreasing=5, interpSpeedIncreasing=5}
    boneToModify {boneName=upperarm_twist_01_r}
    rotation {pitch=-12}
   }
   n8 = "Transform (Modify) Bone - Bone: upperarm_twist_01_l"!(ComponentPose=@n7, Alpha=1, Rotation="12,0,0") -> (Pose) {
    rotationMode=BMM_Additive
    alphaBoolBlend {blendOption=Linear}
    alphaScaleBiasClamp {bInterpResult=true, interpSpeedDecreasing=5, interpSpeedIncreasing=5}
    boneToModify {boneName=upperarm_twist_01_l}
    rotation {pitch=12}
   }
   n9 = "Transform (Modify) Bone - Bone: deform_tricep_r"!(ComponentPose=@n8, Alpha=1, Translation="-3,-2,0", Rotation="-30,10,0") -> (Pose) {
    rotationMode=BMM_Additive
    scaleSpace=BCS_BoneSpace
    translationMode=BMM_Additive
    translationSpace=BCS_BoneSpace
    alphaBoolBlend {blendOption=Linear}
    alphaScaleBiasClamp {bInterpResult=true, interpSpeedDecreasing=5, interpSpeedIncreasing=5}
    boneToModify {boneName=deform_tricep_r}
    rotation {pitch=-30, yaw=10}
    translation=<-3 -2>
   }
   n10 = "Transform (Modify) Bone - Bone: deform_tricep_l"!(ComponentPose=@n9, Alpha=1, Translation="-2,2,0", Rotation="20,10,-4") -> (Pose) {
    rotationMode=BMM_Additive
    translationMode=BMM_Additive
    translationSpace=BCS_BoneSpace
    alphaBoolBlend {blendOption=Linear}
    alphaScaleBiasClamp {bInterpResult=true, interpSpeedDecreasing=5, interpSpeedIncreasing=5}
    boneToModify {boneName=deform_tricep_l}
    rotation=<rot 20 10 -4>
    translation=<-2 2>
   }
   Output Pose(Result=Component To Local(ComponentPose=@n10)) {layerGroup=None}
  ]}
 EventGraph {nodes [
   Event BlueprintUpdateAnimation()
   > Set FortPawn(FortPawn:Fort Pawn Object Reference=TryGetFortPawn()) -> (Output_Get:Fort Pawn Object Reference)
   n1 = Get FortPawn() -> (FortPawn:Fort Pawn Object Reference)
   > n2 = Is Valid(InputObject=@n1)
   n3 = Set WeaponVisible(execute=@n2.Is Not Valid, WeaponVisible:bool=false) -> (Output_Get:bool)
   n4 = Get WeaponVisible() -> (WeaponVisible:bool)
   n5 = Branch(execute=[@n3.then, @n9.then], Condition=@n4)
   n6 = PickaxeItemDefinitionHasAnyMetaTag(Tags=Make Array("[0]"=(TagName="Cosmetics.MetaTags.Clash")), bRequiresEquipped=true)
   > Set ClashPickaxe(ClashPickaxe:double="To Float (Boolean)"(InBool=AND Boolean(A=@n4, B=@n6))) -> (Output_Get:double)
   Set ClashPickaxe(execute=@n5.else, ClashPickaxe=0)
   n7 = IsWeaponHolstered!(self:Fort Pawn Object Reference=@n1) -> (ReturnValue:bool)
   n8 = Get CurrentWeapon!(self:Fort Pawn Object Reference=@n1) -> (CurrentWeapon:Fort Weapon Object Reference)
   n9 = Set WeaponVisible(execute=@n2.Is Valid, WeaponVisible=AND Boolean(A=IsValid(Object=@n8), B=NOT Boolean(A=@n7)))
   Event OnParachuteOpened()
   > Switch on EGliderType(Selection=Get PlayerGliderType() -> (PlayerGliderType=HangGlider))
   > Set HangGliding(HangGliding:double=1) -> (Output_Get:double)
   Event OnParachuteClosed()
   > Set HangGliding(HangGliding=0)
   WeaponDataHasMatchingGameplayTag!(
    self:Fort Weapon Object Reference=Get CurrentWeapon(self=Get FortPawn())
    TagToCheck:Gameplay Tag Structure=(TagName="Cosmetics.MetaTags.Clash")
   ) -> (ReturnValue:bool)
  ]}
]
variables [
 ClashPickaxe {type=number}
 FortPawn {type=FortPawn, default=null}
 HangGliding {type=number}
 WeaponVisible {type=boolean}
]
